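# Unlit translucent Niagara sprite and mesh particle material for the GroovingSparkle player emote, with a Spark parameter blending between two looks.
# At 0, scrolling thin sine stripes along U, faded at both V ends and depth-faded, take a time-driven hue-shifted red.
# At 1, a radial dot with dynamic-parameter hardness takes particle color.
# Particle alpha scales opacity, and emissive passes through MF_FX_EmissiveAdjustments.
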Material /Game/Animation/Game/MainPlayer/Emotes/GroovingSparkle/FX/M_Emote_GroovingSparkle
graph {
 properties {
  bUsedWithNiagaraMeshParticles=true
  bUsedWithNiagaraSprites=true
  blendMode=BLEND_Translucent
  pixelDepthOffsetMode=PDOM_Legacy
  shadingModel=MSM_Unlit
  usedShadingModels=Unlit
 }
 nodes [
  Constant() {r=1}
  n1 = Constant3Vector() {constant=<rgba 1 0 0 0>}
  n2 = ComponentMask(None=TextureCoordinate()) {r=true}
  LinearInterpolate(A=Constant3Vector() {constant=<rgba 0 0 1 0>}, B=@n1, Alpha=Sine(None=Multiply(A=@n2) {constB=0.5}))
  n3 = DynamicParameter()
  LinearInterpolate(Alpha=@n3.R) {constA=1, constB=0}
  n4 = Multiply(A=MaterialFunctionCall() {materialFunction=&DebugTimeSine}.Result) {constB=0.12}
  n5 = ParticleColor()
  MaterialFunctionCall(Emissive=LinearInterpolate(
    A=MaterialFunctionCall(Hue Shift Percentage=@n4, Texture=@n1) {materialFunction=&HueShift}.Result
    B=@n5.RGB
    Alpha=ScalarParameter() {parameterName=Spark}
   )) {materialFunction=&MF_FX_EmissiveAdjustments}
  n6 = TextureCoordinate()
  n7 = Add(A=Multiply(A=ComponentMask(None=@n6) {r=true}) {constB=55}, B=Multiply(A=Time()) {constB=4})
  n8 = ComponentMask(None=TextureCoordinate()) {g=true}
  n9 = Multiply(A=Power(Base=Sine(None=@n7)) {constExponent=100}, B=Sine(None=Multiply(A=@n8) {constB=0.5}))
  n10 = SphereMask(A=TextureCoordinate(), B=Constant() {r=0.5}, Hardness=@n3.R) {attenuationRadius=0.5}
  Multiply(
   A=@n5.A
   B=LinearInterpolate(A=DepthFade(Opacity=@n9), B=Power(Base=@n10) {constExponent=2.4}, Alpha=ScalarParameter() {parameterName=Spark})
  )
  OneMinus(None=@n3.R)
 ]
}
